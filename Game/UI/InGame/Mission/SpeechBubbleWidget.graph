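# Mission HUD speech bubble widget that shows a pawn's current sentence as bold wrapped text centered on a black border.
# InitFromObject casts the passed object to FortPawn and sets BubbleText from that pawn's GetCurrentSentenceSpeechText.
# The text is set only in that init event, with no binding or later refresh in this graph, and a failed cast changes nothing.
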WidgetBlueprint /Game/UI/InGame/Mission/SpeechBubbleWidget
graphs [EventGraph {nodes [
   n1 = Event InitFromObject!() -> (OutputDelegate, then, InitObject:Object Reference) {is=title K2Node_Event}
   > n2 = Cast To FortPawn!(execute, Object=@n1.InitObject) -> (then, CastFailed, AsFort Pawn:Fort Pawn Object Reference) {is=title K2Node_DynamicCast}
   > "SetText (Text)"(
    self=Get BubbleText() -> (BubbleText:Common Text Object Reference)
    InText=GetCurrentSentenceSpeechText!(self:Fort Pawn Object Reference=@n2) -> (ReturnValue:Text)
   )
  ]}]
variables []
widgets {
 info {inheritedWidgetCount=0, parentClass=&CommonUserWidget, rootWidgetClass=&Border, widgetCount=2}
 widgets [
  Border_34 {
   widgetClassPath=&Border
   widget=Border_34
   properties {horizontalAlignment=HAlign_Center, verticalAlignment=VAlign_Center, brushColor=<rgba 0 0 0 1>}
  }
  BubbleText {
   widgetClassPath=&CommonTextBlock
   bIsVariable=true
   parent=Border_34
   widget=BubbleText
   properties {
    style=/FortUILibrary/Styles/Typography/Foundation/TextStyle-Base-XXS-B.TextStyle-Base-XXS-B_C
    text=Text Block
    wrapTextAt=425
    font {
     bForceMonospaced=false
     bMaterialIsStencil=false
     fontMaterial=null
     fontObject=&NotoSans
     letterSpacing=0
     monospacedWidth=1
     size=14
     skewAmount=0
     typefaceFontName=NotoSans-Bold
     outlineSettings {
      bApplyOutlineToDropShadows=false
      bMiteredCorners=false
      bSeparateFillAlpha=false
      outlineMaterial=null
      outlineSize=0
      outlineColor=<rgba 0 0 0 1>
     }
    }
    shadowOffset=<0 0>
   }
   slot {class=BorderSlot, properties {horizontalAlignment=HAlign_Center, verticalAlignment=VAlign_Center}}
  }
 ]
}
